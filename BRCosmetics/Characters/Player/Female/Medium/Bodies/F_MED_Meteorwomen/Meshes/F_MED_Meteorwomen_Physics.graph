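# Physics asset for the F_MED_Meteorwomen female body cosmetic, apparently for secondary motion on its left and right dyn_tube parts.
# Capsules on pelvis and two spine bones form the torso, with each tube's tiny MSTR anchor box constrained to spine_02.
# Spine and anchor joints leave rotation free.
# Each tube is a box hung from its anchor with mirrored placement and tighter swing and twist limits, narrowest in twist.
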PhysicsAsset /BRCosmetics/Characters/Player/Female/Medium/Bodies/F_MED_Meteorwomen/Meshes/F_MED_Meteorwomen_Physics
bodies {
 dyn_tube_01_l [{shapeType=Box, extentX=7.087, extentY=3.5264, extentZ=12.8414, center=<4.0839 0 0>}]
 dyn_tube_01_l_MSTR [{shapeType=Box, extentX=0.01, extentY=0.01, extentZ=0.01}]
 dyn_tube_01_r [{shapeType=Box, extentX=7.087, extentY=3.5264, extentZ=12.8414, center=<-4.0839 0 0>, rotation=<rot 0 0 180>}]
 dyn_tube_01_r_MSTR [{shapeType=Box, extentX=0.01, extentY=0.01, extentZ=0.01}]
 pelvis [{shapeType=Capsule, length=18.2528, radius=7.9318, rotation=<rot -2.7675 -89.9995 0.2906>}]
 spine_01 [{shapeType=Capsule, length=10.1435, radius=8.2468, rotation=<rot 0 -89.9996 -0.0001>}]
 spine_02 [{shapeType=Capsule, length=6.3935, radius=7.2468, rotation=<rot -90 26.565 -116.5658>}]
}
constraints [
 {
  bone1Name=spine_01
  bone2Name=pelvis
  swing1LimitDegrees=45
  swing1Motion=Free
  swing2LimitDegrees=45
  swing2Motion=Free
  twistLimitDegrees=45
  twistMotion=Free
 }
 {
  bone1Name=spine_02
  bone2Name=spine_01
  swing1LimitDegrees=45
  swing1Motion=Free
  swing2LimitDegrees=45
  swing2Motion=Free
  twistLimitDegrees=45
  twistMotion=Free
 }
 {
  bone1Name=dyn_tube_01_l_MSTR
  bone2Name=spine_02
  swing1LimitDegrees=45
  swing1Motion=Free
  swing2LimitDegrees=45
  swing2Motion=Free
  twistLimitDegrees=45
  twistMotion=Free
 }
 {bone1Name=dyn_tube_01_l, bone2Name=dyn_tube_01_l_MSTR, swing1LimitDegrees=35, swing2LimitDegrees=15, twistLimitDegrees=5}
 {
  bone1Name=dyn_tube_01_r_MSTR
  bone2Name=spine_02
  swing1LimitDegrees=45
  swing1Motion=Free
  swing2LimitDegrees=45
  swing2Motion=Free
  twistLimitDegrees=45
  twistMotion=Free
 }
 {bone1Name=dyn_tube_01_r, bone2Name=dyn_tube_01_r_MSTR, swing1LimitDegrees=35, swing2LimitDegrees=15, twistLimitDegrees=5}
]
