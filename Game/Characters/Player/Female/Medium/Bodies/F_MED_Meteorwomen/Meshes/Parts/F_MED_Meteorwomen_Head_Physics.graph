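# Physics asset for the female medium Meteorwomen head part, setting up physics for five helmet cables: two left, two right and one middle.
# A capsule body covers the head, and each cable has a box body plus a tiny MSTR anchor body.
# Each anchor is constrained to the head with free swing and twist, while its cable is constrained to that anchor with tight twist and asymmetric swing limits.
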PhysicsAsset /Game/Characters/Player/Female/Medium/Bodies/F_MED_Meteorwomen/Meshes/Parts/F_MED_Meteorwomen_Head_Physics
bodies {
 dyn_helmetcable1_01_l [{shapeType=Box, extentX=3.5751, extentY=3.7295, extentZ=1.5, center=<2.0183 0 0>}]
 dyn_helmetcable1_01_l_MSTR [{shapeType=Box, extentX=0.01, extentY=0.01, extentZ=0.01}]
 dyn_helmetcable1_01_r [{shapeType=Box, extentX=3.5751, extentY=3.7295, extentZ=1.5, center=<2.0183 0 0>}]
 dyn_helmetcable1_01_r_MSTR [{shapeType=Box, extentX=0.01, extentY=0.01, extentZ=0.01}]
 dyn_helmetcable2_01_l [{shapeType=Box, extentX=3.5751, extentY=3.7295, extentZ=1.5, center=<2.0183 0 0>}]
 dyn_helmetcable2_01_l_MSTR [{shapeType=Box, extentX=0.01, extentY=0.01, extentZ=0.01}]
 dyn_helmetcable2_01_r [{shapeType=Box, extentX=3.5751, extentY=3.7295, extentZ=1.5, center=<2.0183 0 0>}]
 dyn_helmetcable2_01_r_MSTR [{shapeType=Box, extentX=0.01, extentY=0.01, extentZ=0.01}]
 dyn_helmetcable_mid_01 [{shapeType=Box, extentX=3.5751, extentY=3.7295, extentZ=1.5, center=<2.0183 0 0>}]
 dyn_helmetcable_mid_01_MSTR [{shapeType=Box, extentX=0.01, extentY=0.01, extentZ=0.01}]
 head [{shapeType=Capsule, length=1.6547, radius=7.423, center=<3.0713 1.2468 0>, rotation=<rot -90 90 -79.9985>}]
}
constraints [
 {
  bone1Name=dyn_helmetcable1_01_l_MSTR
  bone2Name=head
  swing1LimitDegrees=45
  swing1Motion=Free
  swing2LimitDegrees=45
  swing2Motion=Free
  twistLimitDegrees=45
  twistMotion=Free
 }
 {
  bone1Name=dyn_helmetcable2_01_l_MSTR
  bone2Name=head
  swing1LimitDegrees=45
  swing1Motion=Free
  swing2LimitDegrees=45
  swing2Motion=Free
  twistLimitDegrees=45
  twistMotion=Free
 }
 {
  bone1Name=dyn_helmetcable1_01_r_MSTR
  bone2Name=head
  swing1LimitDegrees=45
  swing1Motion=Free
  swing2LimitDegrees=45
  swing2Motion=Free
  twistLimitDegrees=45
  twistMotion=Free
 }
 {
  bone1Name=dyn_helmetcable2_01_r_MSTR
  bone2Name=head
  swing1LimitDegrees=45
  swing1Motion=Free
  swing2LimitDegrees=45
  swing2Motion=Free
  twistLimitDegrees=45
  twistMotion=Free
 }
 {
  bone1Name=dyn_helmetcable_mid_01_MSTR
  bone2Name=head
  swing1LimitDegrees=45
  swing1Motion=Free
  swing2LimitDegrees=45
  swing2Motion=Free
  twistLimitDegrees=45
  twistMotion=Free
 }
 {bone1Name=dyn_helmetcable1_01_l, bone2Name=dyn_helmetcable1_01_l_MSTR, swing1LimitDegrees=10, swing2LimitDegrees=30, twistLimitDegrees=2}
 {bone1Name=dyn_helmetcable2_01_l, bone2Name=dyn_helmetcable2_01_l_MSTR, swing1LimitDegrees=10, swing2LimitDegrees=30, twistLimitDegrees=2}
 {bone1Name=dyn_helmetcable1_01_r, bone2Name=dyn_helmetcable1_01_r_MSTR, swing1LimitDegrees=10, swing2LimitDegrees=30, twistLimitDegrees=2}
 {bone1Name=dyn_helmetcable2_01_r, bone2Name=dyn_helmetcable2_01_r_MSTR, swing1LimitDegrees=10, swing2LimitDegrees=30, twistLimitDegrees=2}
 {bone1Name=dyn_helmetcable_mid_01, bone2Name=dyn_helmetcable_mid_01_MSTR, swing1LimitDegrees=10, swing2LimitDegrees=30, twistLimitDegrees=2}
]
