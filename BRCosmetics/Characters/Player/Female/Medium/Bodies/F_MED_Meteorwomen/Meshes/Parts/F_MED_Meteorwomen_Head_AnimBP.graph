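# Animation Blueprint for the Meteorwomen female medium head cosmetic part, adding secondary motion to its helmet cables.
# It copies the pose from the attached parent mesh, then chains five Anim Dynamics nodes for left, right and middle cable bones.
# Each simulates relative to the attach bone with angular springs, angular limits, heavy damping and reduced gravity.
# A RigidBody node sits unconnected, and the EventGraph holds no logic.
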AnimBlueprint /BRCosmetics/Characters/Player/Female/Medium/Bodies/F_MED_Meteorwomen/Meshes/Parts/F_MED_Meteorwomen_Head_AnimBP
target_skeleton=&/F_MED_Meteorwomen_Head_Skeleton
graphs [
 AnimGraph {nodes [
   RigidBody() {
    bClampLinearTranslationLimitToRefPose=true
    bForceDisableCollisionBetweenConstraintBodies=true
    bOverrideWorldGravity=true
    overridePhysicsAsset=/Game/Characters/Player/Female/Medium/Bodies/F_MED_Meteorwomen/Meshes/Parts/F_MED_Meteorwomen_Head_Physics
    simulationSpace=BaseBoneSpace
    alphaBoolBlend {blendOption=Linear}
    baseBoneRef {boneName=neck_01}
    componentAppliedLinearAccClamp=<3000 3000 3000>
    componentLinearAccScale=<0.3 0.3 0.6>
    componentLinearVelScale=<0.3 0.3 0.6>
    overrideWorldGravity {z=-980}
   }
   n1 = Anim Dynamics - Bone: dyn_helmetcable1_01_l(ComponentPose=Local To Component(LocalPose=Copy Pose From Mesh() {bUseAttachedParent=true}), Alpha=1, ExternalForce="0,0,0") {
    angularDampingOverride=0.9999
    angularSpringConstant=500
    bAngularSpring=true
    bGravityOverrideInSimSpace=true
    bOverrideAngularDamping=true
    bUseGravityOverride=true
    gravityScale=0.5
    simulationSpace=BoneRelative
    alphaBoolBlend {blendOption=Linear}
    boundBone {boneName=dyn_helmetcable1_01_l}
    componentAppliedLinearAccClamp=<5000 5000 5000>
    componentLinearAccScale=<0.2 0.2 0.4>
    componentLinearVelScale=<0.5 0.5 0.5>
    gravityOverride {z=-980}
    physicsBodyDefinitions [{
      sphereCollisionRadius=0
      boundBone {boneName=dyn_helmetcable1_01_l}
      boxExtents=<20 20 10>
      constraintSetup {angularLimitsMax=<0.5 3 15>, angularLimitsMin=<-0.5 -30 -15>, angularTarget=<0 0 0>}
      localJointOffset=<0 0 0>
     }]
    relativeSpaceBone {boneName=attach}
   }
   n2 = Anim Dynamics - Bone: dyn_helmetcable2_01_l(ComponentPose=@n1, Alpha=1, ExternalForce="0,0,0") {
    angularDampingOverride=0.9999
    angularSpringConstant=500
    bAngularSpring=true
    bGravityOverrideInSimSpace=true
    bOverrideAngularDamping=true
    bUseGravityOverride=true
    gravityScale=0.5
    simulationSpace=BoneRelative
    alphaBoolBlend {blendOption=Linear}
    boundBone {boneName=dyn_helmetcable2_01_l}
    componentAppliedLinearAccClamp=<5000 5000 5000>
    componentLinearAccScale=<0.2 0.2 0.4>
    componentLinearVelScale=<0.5 0.5 0.5>
    gravityOverride {z=-980}
    physicsBodyDefinitions [{
      sphereCollisionRadius=0
      boundBone {boneName=dyn_helmetcable2_01_l}
      boxExtents=<20 20 10>
      constraintSetup {angularLimitsMax=<0.5 3 15>, angularLimitsMin=<-0.5 -30 -15>, angularTarget=<0 0 0>}
      localJointOffset=<0 0 0>
     }]
    relativeSpaceBone {boneName=attach}
   }
   n3 = Anim Dynamics - Bone: dyn_helmetcable1_01_r(ComponentPose=@n2, Alpha=1, ExternalForce="0,0,0") {
    angularDampingOverride=0.9999
    angularSpringConstant=500
    bAngularSpring=true
    bGravityOverrideInSimSpace=true
    bOverrideAngularDamping=true
    bUseGravityOverride=true
    gravityScale=0.5
    simulationSpace=BoneRelative
    alphaBoolBlend {blendOption=Linear}
    boundBone {boneName=dyn_helmetcable1_01_r}
    componentAppliedLinearAccClamp=<5000 5000 5000>
    componentLinearAccScale=<0.2 0.2 0.4>
    componentLinearVelScale=<0.5 0.5 0.5>
    gravityOverride {z=-980}
    physicsBodyDefinitions [{
      sphereCollisionRadius=0
      boundBone {boneName=dyn_helmetcable1_01_r}
      boxExtents=<20 20 10>
      constraintSetup {angularLimitsMax=<0.5 3 15>, angularLimitsMin=<-0.5 -30 -15>, angularTarget=<0 0 0>}
      localJointOffset=<3 0 -5>
     }]
    relativeSpaceBone {boneName=attach}
   }
   n4 = Anim Dynamics - Bone: dyn_helmetcable2_01_r(ComponentPose=@n3, Alpha=1, ExternalForce="0,0,0") {
    angularDampingOverride=0.9999
    angularSpringConstant=500
    bAngularSpring=true
    bGravityOverrideInSimSpace=true
    bOverrideAngularDamping=true
    bUseGravityOverride=true
    gravityScale=0.5
    simulationSpace=BoneRelative
    alphaBoolBlend {blendOption=Linear}
    boundBone {boneName=dyn_helmetcable2_01_r}
    componentAppliedLinearAccClamp=<5000 5000 5000>
    componentLinearAccScale=<0.2 0.2 0.4>
    componentLinearVelScale=<0.5 0.5 0.5>
    gravityOverride {z=-980}
    physicsBodyDefinitions [{
      sphereCollisionRadius=0
      boundBone {boneName=dyn_helmetcable2_01_r}
      boxExtents=<20 20 10>
      constraintSetup {angularLimitsMax=<0.5 3 15>, angularLimitsMin=<-0.5 -30 -15>, angularTarget=<0 0 0>}
      localJointOffset=<3 0 -5>
     }]
    relativeSpaceBone {boneName=attach}
   }
   n5 = Anim Dynamics - Bone: dyn_helmetcable_mid_01(ComponentPose=@n4, Alpha=1, ExternalForce="0,0,0") {
    angularDampingOverride=0.9999
    angularSpringConstant=500
    bAngularSpring=true
    bGravityOverrideInSimSpace=true
    bOverrideAngularDamping=true
    bUseGravityOverride=true
    gravityScale=0.5
    simulationSpace=BoneRelative
    alphaBoolBlend {blendOption=Linear}
    boundBone {boneName=dyn_helmetcable_mid_01}
    componentAppliedLinearAccClamp=<5000 5000 5000>
    componentLinearAccScale=<0.2 0.2 0.4>
    componentLinearVelScale=<0.5 0.5 0.5>
    gravityOverride {z=-980}
    physicsBodyDefinitions [{
      sphereCollisionRadius=0
      boundBone {boneName=dyn_helmetcable_mid_01}
      boxExtents=<20 20 10>
      constraintSetup {angularLimitsMax=<0.5 3 15>, angularLimitsMin=<-0.5 -30 -15>, angularTarget=<0 0 0>}
      localJointOffset=<3 0 -5>
     }]
    relativeSpaceBone {boneName=attach}
   }
   Output Pose(Result=Component To Local(ComponentPose=@n5)) {layerGroup=None}
  ]}
 EventGraph {nodes [
   TryGetPawnOwner()
   Event BlueprintUpdateAnimation()
  ]}
]
variables []
